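# Physics asset for the Weeping Woods Toon Male pickaxe, a flower-shaped two-handed melee cosmetic with dynamic stem, petal and bee bones.
# Capsule bodies chain from the root through a fully locked handle and handle top into a two-segment bending stem, which carries the flower center and five four-segment petals.
# Stem and petals use tight swing limits with twist locked on the stem and petal bases, while the bee body rotates freely about the flower center.
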PhysicsAsset /BRCosmetics/Weapons/FORT_Melee/Pickaxe_Weeping_Woods_Toon_Male/Meshes/Weeping_Woods_Toon_Male_Axe_Physics1
bodies {
 Melee_TwoHanded_Handle [{shapeType=Capsule, length=2, radius=2, shapeName=Melee_TwoHanded_Handle_capsule}]
 b_AR_Root [{shapeType=Capsule, length=0.505, radius=0.505, shapeName=b_AR_Root_capsule, rotation=<rot 0 -90 0>}]
 dyn_FlowerCenter [{shapeType=Capsule, length=2, radius=2, shapeName=dyn_FlowerCenter_capsule}]
 dyn_Petal1_01 [{shapeType=Capsule, length=2, radius=2, shapeName=dyn_Petal1_01_capsule}]
 dyn_Petal1_02 [{shapeType=Capsule, length=2, radius=2, shapeName=dyn_Petal1_02_capsule}]
 dyn_Petal1_03 [{shapeType=Capsule, length=2, radius=2, shapeName=dyn_Petal1_03_capsule}]
 dyn_Petal1_04 [{shapeType=Capsule, length=2, radius=2, shapeName=dyn_Petal1_04_capsule}]
 dyn_Petal2_01 [{shapeType=Capsule, length=2, radius=2, shapeName=dyn_Petal2_01_capsule}]
 dyn_Petal2_02 [{shapeType=Capsule, length=2, radius=2, shapeName=dyn_Petal2_02_capsule}]
 dyn_Petal2_03 [{shapeType=Capsule, length=2, radius=2, shapeName=dyn_Petal2_03_capsule}]
 dyn_Petal2_04 [{shapeType=Capsule, length=2, radius=2, shapeName=dyn_Petal2_04_capsule}]
 dyn_Petal3_01 [{shapeType=Capsule, length=2, radius=2, shapeName=dyn_Petal3_01_capsule}]
 dyn_Petal3_02 [{shapeType=Capsule, length=2, radius=4.25, shapeName=dyn_Petal3_02_capsule, rotation=<rot 79.9999 -179.9999 -179.9999>}]
 dyn_Petal3_03 [{shapeType=Capsule, length=2, radius=4.75, shapeName=dyn_Petal3_03_capsule, center=<-2.7833 -2.2226 -8.5726>, rotation=<rot 90 0 -10.0001>}]
 dyn_Petal3_04 [{shapeType=Capsule, length=2, radius=3.75, shapeName=dyn_Petal3_04_capsule, rotation=<rot 80.0002 0 0>}]
 dyn_Petal4_01 [{shapeType=Capsule, length=2, radius=2, shapeName=dyn_Petal4_01_capsule}]
 dyn_Petal4_02 [{shapeType=Capsule, length=2, radius=5.5, shapeName=dyn_Petal4_02_capsule, rotation=<rot 79.9998 -179.9998 -179.9998>}]
 dyn_Petal4_03 [{
   shapeType=Capsule
   length=2
   radius=4.75
   shapeName=dyn_Petal4_03_capsule
   center=<-3.0101 4.8807 2.2025>
   rotation=<rot -69.9997 179.9998 179.9996>
  }]
 dyn_Petal4_04 [{shapeType=Capsule, length=2, radius=3.75, shapeName=dyn_Petal4_04_capsule, rotation=<rot 80.0009 0.0001 0>}]
 dyn_Petal5_01 [{shapeType=Capsule, length=2, radius=2, shapeName=dyn_Petal5_01_capsule}]
 dyn_Petal5_02 [{shapeType=Capsule, length=2, radius=2, shapeName=dyn_Petal5_02_capsule}]
 dyn_Petal5_03 [{shapeType=Capsule, length=2, radius=2, shapeName=dyn_Petal5_03_capsule}]
 dyn_Petal5_04 [{shapeType=Capsule, length=2, radius=2, shapeName=dyn_Petal5_04_capsule}]
 dyn_StemBend01 [{shapeType=Capsule, length=2, radius=2, shapeName=dyn_StemBend01_capsule}]
 dyn_StemBend02 [{shapeType=Capsule, length=2, radius=2, shapeName=dyn_StemBend02_capsule}]
 dyn_TopOfHandle [{shapeType=Capsule, length=2, radius=2, shapeName=dyn_TopOfHandle_capsule}]
 dyn_beeBody [{shapeType=Capsule, length=1, radius=1, shapeName=dyn_beeBody_capsule}]
}
constraints [
 {
  bone1Name=dyn_TopOfHandle
  bone2Name=Melee_TwoHanded_Handle
  swing1LimitDegrees=10
  swing1Motion=Locked
  swing2LimitDegrees=10
  swing2Motion=Locked
  twistLimitDegrees=10
  twistMotion=Locked
 }
 {bone1Name=dyn_StemBend01, bone2Name=dyn_TopOfHandle, swing1LimitDegrees=10, swing2LimitDegrees=10, twistLimitDegrees=10, twistMotion=Locked}
 {bone1Name=dyn_StemBend02, bone2Name=dyn_StemBend01, swing1LimitDegrees=10, swing2LimitDegrees=10, twistLimitDegrees=10, twistMotion=Locked}
 {bone1Name=dyn_FlowerCenter, bone2Name=dyn_StemBend02, swing1LimitDegrees=10, swing2LimitDegrees=10, twistLimitDegrees=10, twistMotion=Locked}
 {bone1Name=dyn_Petal1_01, bone2Name=dyn_FlowerCenter, swing1LimitDegrees=5, swing2LimitDegrees=5, twistLimitDegrees=10, twistMotion=Locked}
 {bone1Name=dyn_Petal1_02, bone2Name=dyn_Petal1_01, swing1LimitDegrees=5, swing2LimitDegrees=5, twistLimitDegrees=10, twistMotion=Locked}
 {bone1Name=dyn_Petal1_03, bone2Name=dyn_Petal1_02, swing1LimitDegrees=5, swing2LimitDegrees=5, twistLimitDegrees=10}
 {bone1Name=dyn_Petal1_04, bone2Name=dyn_Petal1_03, swing1LimitDegrees=5, swing2LimitDegrees=5, twistLimitDegrees=10}
 {bone1Name=dyn_Petal2_01, bone2Name=dyn_FlowerCenter, swing1LimitDegrees=5, swing2LimitDegrees=5, twistLimitDegrees=10, twistMotion=Locked}
 {bone1Name=dyn_Petal2_02, bone2Name=dyn_Petal2_01, swing1LimitDegrees=5, swing2LimitDegrees=5, twistLimitDegrees=10}
 {bone1Name=dyn_Petal2_03, bone2Name=dyn_Petal2_02, swing1LimitDegrees=5, swing2LimitDegrees=5, twistLimitDegrees=10}
 {bone1Name=dyn_Petal2_04, bone2Name=dyn_Petal2_03, swing1LimitDegrees=5, swing2LimitDegrees=5, twistLimitDegrees=10}
 {bone1Name=dyn_Petal3_01, bone2Name=dyn_FlowerCenter, swing1LimitDegrees=5, swing2LimitDegrees=5, twistLimitDegrees=10, twistMotion=Locked}
 {bone1Name=dyn_Petal3_02, bone2Name=dyn_Petal3_01, swing1LimitDegrees=5, swing2LimitDegrees=5, twistLimitDegrees=10}
 {bone1Name=dyn_Petal3_03, bone2Name=dyn_Petal3_02, swing1LimitDegrees=5, swing2LimitDegrees=5, twistLimitDegrees=10}
 {bone1Name=dyn_Petal3_04, bone2Name=dyn_Petal3_03, swing1LimitDegrees=5, swing2LimitDegrees=5, twistLimitDegrees=10}
 {bone1Name=dyn_Petal4_01, bone2Name=dyn_FlowerCenter, swing1LimitDegrees=5, swing2LimitDegrees=5, twistLimitDegrees=10, twistMotion=Locked}
 {bone1Name=dyn_Petal4_02, bone2Name=dyn_Petal4_01, swing1LimitDegrees=5, swing2LimitDegrees=5, twistLimitDegrees=10}
 {bone1Name=dyn_Petal4_03, bone2Name=dyn_Petal4_02, swing1LimitDegrees=5, swing2LimitDegrees=5, twistLimitDegrees=10}
 {bone1Name=dyn_Petal4_04, bone2Name=dyn_Petal4_03, swing1LimitDegrees=5, swing2LimitDegrees=5, twistLimitDegrees=10}
 {bone1Name=dyn_Petal5_01, bone2Name=dyn_FlowerCenter, swing1LimitDegrees=5, swing2LimitDegrees=5, twistLimitDegrees=10, twistMotion=Locked}
 {bone1Name=dyn_Petal5_02, bone2Name=dyn_Petal5_01, swing1LimitDegrees=5, swing2LimitDegrees=5, twistLimitDegrees=10}
 {bone1Name=dyn_Petal5_03, bone2Name=dyn_Petal5_02, swing1LimitDegrees=5, swing2LimitDegrees=5, twistLimitDegrees=10}
 {bone1Name=dyn_Petal5_04, bone2Name=dyn_Petal5_03, swing1LimitDegrees=5, swing2LimitDegrees=5, twistLimitDegrees=10}
 {
  bone1Name=Melee_TwoHanded_Handle
  bone2Name=b_AR_Root
  swing1LimitDegrees=45
  swing1Motion=Locked
  swing2LimitDegrees=45
  swing2Motion=Locked
  twistLimitDegrees=45
  twistMotion=Locked
 }
 {
  bone1Name=dyn_beeBody
  bone2Name=dyn_FlowerCenter
  swing1LimitDegrees=75
  swing1Motion=Free
  swing2LimitDegrees=75
  swing2Motion=Free
  twistLimitDegrees=45
  twistMotion=Free
 }
]
