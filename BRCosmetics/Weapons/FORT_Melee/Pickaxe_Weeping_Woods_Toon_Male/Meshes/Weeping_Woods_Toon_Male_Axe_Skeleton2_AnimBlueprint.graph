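# Animation Blueprint for the Weeping Woods Toon Male pickaxe, playing its idle animation with simulated secondary motion on the bee.
# The idle feeds a RigidBody node set to zero alpha, then a spring-driven, constrained Anim Dynamics node on dyn_beeBody.
# The event graph stores a tracking curve and sets physics alpha variables by frontend display context, but no visible pose node uses them.
# A second, world-space Anim Dynamics node sits unconnected.
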AnimBlueprint /BRCosmetics/Weapons/FORT_Melee/Pickaxe_Weeping_Woods_Toon_Male/Meshes/Weeping_Woods_Toon_Male_Axe_Skeleton2_AnimBlueprint
target_skeleton=&/Weeping_Woods_Toon_Male_Axe_Skeleton2
graphs [
 AnimGraph {nodes [
   Anim Dynamics - Bone: dyn_beeBody(Alpha=1, ExternalForce="0,0,0") {
    bUseGravityOverride=true
    simulationSpace=World
    alphaBoolBlend {blendOption=Linear}
    boundBone {boneName=dyn_beeBody}
    componentAppliedLinearAccClamp=<3000 3000 3000>
    componentLinearAccScale=<0.4 0.4 0.4>
    componentLinearVelScale=<0.4 0.4 0.4>
    gravityOverride {z=-980}
    physicsBodyDefinitions [{
      sphereCollisionRadius=0
      boundBone {boneName=dyn_beeBody}
      boxExtents=<20 20 20>
      constraintSetup {angularLimitsMax=<100 100 100>, angularLimitsMin=<-100 -100 -100>, angularTarget=<0 0 0>, linearAxesMax=<20 20 20>, linearAxesMin=<-5 -5 -5>}
      localJointOffset=<0 0 0>
     }]
    relativeSpaceBone {boneName=dyn_FlowerCenter}
   }
   Input Pose()
   n1 = RigidBody(ComponentPose=Local To Component(LocalPose=Sequence Player 'Weeping_Woods_Toon_Male_Axe_Idle'() {sequence=/BRCosmetics/Weapons/FORT_Melee/Pickaxe_Weeping_Woods_Toon_Male/Animation/Weeping_Woods_Toon_Male_Axe_Idle})) {
    alpha=0
    bOverrideWorldGravity=true
    overridePhysicsAsset=&/Weeping_Woods_Toon_Male_Axe_Physics1
    simulationSpace=BaseBoneSpace
    alphaBoolBlend {blendOption=Linear}
    baseBoneRef {boneName=Melee_TwoHanded_Handle}
    componentAppliedLinearAccClamp=<3000 3000 3000>
    componentLinearAccScale=<0.4 0.3 0.3>
    componentLinearVelScale=<0.4 0.3 0.3>
    overrideWorldGravity {z=-980}
   }
   n2 = Anim Dynamics - Bone: dyn_beeBody(ComponentPose=@n1, Alpha=1, ExternalForce="0,0,0") {
    bAngularSpring=true
    bLinearSpring=true
    bOverrideAngularBias=true
    bOverrideAngularDamping=true
    bUseGravityOverride=true
    linearSpringConstant=15
    simulationSpace=Actor
    alphaBoolBlend {blendOption=Linear}
    boundBone {boneName=dyn_beeBody}
    gravityOverride=<1 1 1>
    physicsBodyDefinitions [{
      sphereCollisionRadius=0
      boundBone {boneName=dyn_beeBody}
      boxExtents=<60 60 60>
      constraintSetup {
       angularTargetAxis=AxisY
       coneAngle=60
       twistAxis=AxisY
       angularLimitsMax=<20 0 20>
       angularLimitsMin=<-20 0 -30>
       angularTarget=<0 1 0>
       linearAxesMax=<20 0 20>
       linearAxesMin=<-20 -10 -20>
      }
      localJointOffset=<0 -2 0>
     }]
   }
   Output Pose(Result=Component To Local(ComponentPose=@n2)) {layerGroup=None}
   Get Anim Tracking() -> (Anim Tracking:double)
  ]}
 EventGraph {nodes [
   Event BlueprintUpdateAnimation()
   > Set Anim Tracking(Anim Tracking:double=GetCurveValue(CurveName=WeepingWoodsAnimTracking)) -> (Output_Get:double)
   > n1 = Branch(Condition=IsCosmeticObjectDisplayedInFrontendContext!(CosmeticObject:Object Reference) -> (ReturnValue:bool))
   > Set dynAlpha(dynAlpha:double=0.25) -> (Output_Get:double)
   > Set rigidAlpha(rigidAlpha:double=0) -> (Output_Get:double)
   Set dynAlpha(execute=@n1.else, dynAlpha=1)
   > Set rigidAlpha(rigidAlpha=1)
   TryGetPawnOwner()
  ]}
]
variables [
 Anim Tracking {type=number}
 dynAlpha {type=number}
 rigidAlpha {type=number}
]
